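; Wide TV prop Blueprint that switches static screen effects and a looping glitch sound according to its Void Slide state.
; BeginPlay and RepNotify on bVoidSlideActivated toggle the TVStatic_On material scalar and the audio loop.
; A separate bonus RepNotify toggles TVStatic_Bonus, with a timer clearing the effect after BonusTimeToTVDeactivate.
; All effects run only where cosmetic events can execute, and each prints a debug string.
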(event EventBeginPlay
  (|CallParentFunction)
  (if (Variables|Default|GetbVoidSlideActivated)
    (CallFunction|VoidSlideActivated)
    (else
      (CallFunction|VoidSlideDeactivated))))

(event Custom|VoidSlideActivated
  ;[1 1 1] Showcase for hookup to event
  (Networking|CanExecuteCosmeticEvents
    (:True
      (Rendering|Material|SetScalarParameterValueonMaterials (Variables|Debug|GetStaticMeshComponent) "TVStatic_On" 1.0)
      (Development|PrintString "TV is Powered ON!")
      (Class|FortPropAudioComponent|PlayAudio (Variables|Default|GetFortPropAudioComponent_TV_Glitch_Loop)))) ; client?
)

(event Custom|VoidSlideDeactivated
  ;[1 1 1] Showcase for hookup to event
  (Networking|CanExecuteCosmeticEvents
    (:True
      (Rendering|Material|SetScalarParameterValueonMaterials (Variables|Debug|GetStaticMeshComponent) "TVStatic_On" 0.0)
      (Development|PrintString "TV is Powered OFF!")
      (Class|FortPropAudioComponent|StopAudio :self (Variables|Default|GetFortPropAudioComponent_TV_Glitch_Loop) :Damage 0.0 :Momentum "0, 0, 0"))) ; client?
)

(event Custom|VoidSlideBonusActivated
  (Networking|CanExecuteCosmeticEvents
    (:True
      (Utilities|Time|SetTimerbyEvent (EventDispatchers|CreateEvent :!event Custom|VoidSlideBonusDeactivated) (Variables|Default|GetBonusTimeToTVDeactivate)) ; Set timer to deactivate Bonus mode
      ;[1 0.8378 0.06719] Set material attribute to control the bonus TV moment
      (Rendering|Material|SetScalarParameterValueonMaterials (Variables|Debug|GetStaticMeshComponent) "TVStatic_Bonus" 1.0)
      (Development|PrintString "TV isBONUSed"))) ; client?
)

(event Custom|VoidSlideBonusDeactivated
  (Networking|CanExecuteCosmeticEvents
    (:True
      (Rendering|Material|SetScalarParameterValueonMaterials (Variables|Debug|GetStaticMeshComponent) "TVStatic_Bonus" 0.0)
      (Development|PrintString "TV isNotBONUSed"))) ; client?
)

(fn OnRep_bVoidSlideActivated
  (if (Variables|Default|GetbVoidSlideActivated)
    (CallFunction|VoidSlideActivated)
    (else
      (CallFunction|VoidSlideDeactivated))))

(fn OnRep_bVoidSlideBonusActivated
  (if (Variables|Default|GetbVoidSlideBonusActivated)
    (CallFunction|VoidSlideBonusActivated)
    (else
      (CallFunction|VoidSlideBonusDeactivated))))

(fn ConstructionScript
  (|CallParentFunction))
